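# Niagara system for the Plot Twist emote, adding electric arcs, thruster fire and ground-level smoke, dust and wind to a skeletal mesh.
# Ribbons form Bezier splines between two sampled sockets, with flickering flare lights at the ends and jittered secondary strands.
# Mirrored bone-attached flames and ground effects follow System Timeline curves, while exhaust spawns by movement distance.
# User.bIsInWater zeroes ground spawn rates and exhaust spacing; time-of-day functions adjust smoke and streaks.
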NiagaraSystem /BRCosmetics/Effects/Fort_Effects/Emotes/PlotTwist/NS_PlotTwist_Emote
dependencies {
 usedDataInterfaces [&NiagaraDataInterfaceCurve, &NiagaraDataInterfaceColorCurve]
 usedDynamicInputs [
  &AddVectorToPosition
  &Boolean_OrOperation
  &SetBoolByFloatComparison
  &Abs_Float
 ]
 usedModules [&SystemState, &Timeline]
}
emitters {
 Electrical_Connector_Flares {
  emitterUpdateScript [
   &EmitterState()
   &SpawnBurst_Instantaneous(Spawn Count=2, Use Spawn Probability=false, UseLoopCountLimit=true)
  ]
  particleSpawnScript [
   &InitializeParticle(
    Color=<rgba 10 1.1458 9.2717 0.2>
    Lifetime=0.1
    Sprite Rotation Mode=::NewEnumerator2
    Sprite Size Mode=::NewEnumerator3
    Uniform Sprite Size=12
    UsePositionOffset=false
    Write Material Random=true
   )
   &CameraOffset(Camera Offset Amount=5)
  ]
  particleUpdateScript [
   &ParticleState(
    Kill Particles When Lifetime Has Elapsed=false
    Let Infinitely Lived Particles Die When Emitter Deactivates=true
    Loop Particles Lifetime=true
   )
   &SkeletalMeshLocation(
    Bone / Socket Index=&ReturnExecIndex()
    MaskSpawnGroup=false
    Mesh Sampling Type=::NewEnumerator5
    Skeletal Mesh=$User.SkeletalMesh
    Socket Sampling Mode=::NewEnumerator5
   )
   &SpriteRotationRate(Rotation Rate=360)
   &ScaleSpriteSize(Uniform Curve Sprite Scale=DataInterface() {propertyValues {Curve {keys [
        {arriveTangent=3.3089, interpMode=RCIM_Cubic, time=0, value=1}
        {arriveTangent=0.2947, interpMode=RCIM_Cubic, leaveTangent=0.2947, time=0.1484, value=0.9224}
        {arriveTangent=-0.0227, interpMode=RCIM_Cubic, leaveTangent=-0.0227, tangentMode=RCTM_User, time=0.3217, value=1.0948}
        {arriveTangent=-0.6067, interpMode=RCIM_Cubic, leaveTangent=-0.6067, time=0.3841, value=0.7472}
        {arriveTangent=0.6411, interpMode=RCIM_Cubic, leaveTangent=0.6411, time=0.5191, value=0.975}
        {arriveTangent=0.3939, interpMode=RCIM_Cubic, leaveTangent=0.3939, time=0.6517, value=0.9188}
        {arriveTangent=-0.5293, interpMode=RCIM_Cubic, leaveTangent=-0.5293, time=0.7648, value=1.0718}
        {arriveTangent=-0.305, interpMode=RCIM_Cubic, leaveTangent=-0.305, time=0.8187, value=0.8304}
        {interpMode=RCIM_Cubic, leaveTangent=-0.0913, time=1, value=1}
       ]}}})
   &ScaleColor(
    Scale Alpha=&RandomRangeFloat(Minimum=0.2, Maximum=0.5, Evaluation Type=::NewEnumerator1)
    ScaleA=true
    ScaleRGB=true
   )
   &FN_Light_Attributes(
    Color=$Particles.Initial.Color
    Create Color Attribute=true
    Create Exponent Attribute=true
    Create Radius Attribute=true
    Exponent=4.5
    Radius=100
    Scale Color=&Waveform(
     "Waveform [1]"=::NewEnumerator12
     "[1] Frequency"=&RandomRangeFloat(Minimum=20, Maximum=22)
     "[1] Amplitude Min/Max"=<0.25 1>
    )
   )
  ]
  rendererClasses [&NiagaraSpriteRendererProperties, &NiagaraLightRendererProperties]
  renderers [
   {rendererClass=&NiagaraSpriteRendererProperties}
   {rendererClass=&NiagaraLightRendererProperties, rendererIndex=1}
  ]
 }
 Electrical_Current {
  emitterUpdateScript [
   &EmitterState()
   &SpawnBurst_Instantaneous(Spawn Count=20, Use Spawn Probability=false, UseLoopCountLimit=true)
  ]
  particleSpawnScript [
   &InitializeParticle(
    Color=<rgba 3.2703 0.7083 4 3>
    Lifetime=0.75
    Ribbon Width=12
    Ribbon Width Mode=::NewEnumerator1
    Sprite Size Mode=::NewEnumerator3
    Uniform Sprite Size=20
    UsePositionOffset=false
    Write Material Random=true
   )
   &PartitionParticles(
    Normalized Index In Partition Includes One=true
    Particles Per Partition=20
    Partition Count Mode:ENiagara_InfiniteLoopDuration=::NewEnumerator1
    Partition Mode:ENiagara_PartitionMode=::NewEnumerator1
    Write Partition Index to Ribbon Link Order:ENiagara_AttributeSamplingApplyOutput=::NewEnumerator0
    Write Partition to Ribbon ID=::NewEnumerator0
   )
   &DynamicMaterialParameters(
    Index 0 Param 2=1
    Index 0 Param 3=0.4
    Index 1 Param 1=3
    Index 1 Param 2=1
    Index 1 Param 3=1
    Index 1 Param 4=1
    Module.Param0WriteEnabled=true
    Module.Param0WriteEnabled1=true
    Module.Param1WriteEnabled=true
    Module.Param1WriteEnabled1=true
    Module.Param2WriteEnabled=true
    Module.Param2WriteEnabled1=true
    Module.Param3WriteEnabled=true
    Module.Param3WriteEnabled1=true
    Write Parameter Index 1=true
   )
  ]
  particleUpdateScript [
   &ParticleState(
    Kill Particles When Lifetime Has Elapsed=false
    Let Infinitely Lived Particles Die When Emitter Deactivates=true
    Loop Particles Lifetime=true
   )
   &SkeletalMeshLocation(
    MaskSpawnGroup=false
    Mesh Sampling Type=::NewEnumerator5
    Skeletal Mesh=$User.SkeletalMesh
    Socket Sampling Mode=::NewEnumerator5
    Write To Particle Position=false
   )
   &SkeletalMeshLocation(
    Bone / Socket Index=1
    MaskSpawnGroup=false
    Mesh Sampling Type=::NewEnumerator5
    Skeletal Mesh=$User.SkeletalMesh
    Socket Sampling Mode=::NewEnumerator5
    Write To Particle Position=false
   ) {module=SkeletalMeshLocation001}
   /SetVariables(Particles.MidPoint:Vector3f=&Lerp_Vector(A=$Particles.SkeletalMeshLocation.SampledPosition, B=$Particles.SkeletalMeshLocation001.SampledPosition, Alpha=0.5)) {isSetParameters=true, module=SetVariables_#1}
   &self:BezierSpline(
    EndHandle:Vector3f=&Lerp_Vector(A=$Particles.SkeletalMeshLocation001.SampledPosition, B=$Particles.MidPoint, Alpha=0.5)
    EndPosition:Vector3f=$Particles.SkeletalMeshLocation001.SampledPosition
    PositionAlongSpline:NiagaraFloat=$Particles.PartitionParticles.Normalized Index In Partition
    StartHandle:Vector3f=&Lerp_Vector(A=$Particles.SkeletalMeshLocation.SampledPosition, B=$Particles.MidPoint, Alpha=0.5)
    StartPosition:Vector3f=$Particles.SkeletalMeshLocation.SampledPosition
   ) {module=BezierSpline}
   /SetVariables(Particles.Position:Vector3f=$Particles.Spline_Position) {isSetParameters=true, module=SetVariables_#2}
   &ScaleColor(
    Scale RGB=&VectorFromFloat(Value=&FloatFromCurve(FloatCurve=DataInterface() {propertyValues {Curve {keys [<at 0 1>,<at 0.1 15>,<at 0.2 1>,<at 0.3 1>,<at 0.4 9>,<at 0.5 1>,<at 0.6 3>,<at 0.7 1>,<at 0.8 6>,<at 0.9 1>,<at 1 1>]}}}))
    ScaleA=true
    ScaleRGB=true
   )
   &ScaleRibbonWidth(Ribbon Width Scale=&FloatFromCurve(
     FloatCurve=DataInterface() {propertyValues {Curve {keys [
         {arriveTangent=3.69, interpMode=RCIM_Cubic, leaveTangent=3.69, tangentMode=RCTM_User, time=0, value=0.25}
         {interpMode=RCIM_Cubic, tangentMode=RCTM_User, time=0.5, value=1}
         {arriveTangent=-3.5528, interpMode=RCIM_Cubic, leaveTangent=-3.5528, tangentMode=RCTM_User, time=1, value=0.25}
        ]}}}
     CurveIndex=$Particles.RibbonLinkOrder
    ))
   &FN_Sprite_Attributes(
    Color=<rgba 0.437 0.0159 0.3999 0.15>
    Create Color Attribute=true
    Create SpriteSize Attribute=true
    SpriteSize=&Vector2DFromFloat(Value=&FloatFromCurve(
      FloatCurve=DataInterface() {propertyValues {Curve {keys [
          {arriveTangent=3.69, interpMode=RCIM_Cubic, leaveTangent=3.69, tangentMode=RCTM_User, time=0, value=0.25}
          {interpMode=RCIM_Cubic, tangentMode=RCTM_User, time=0.5, value=1}
          {arriveTangent=-3.5528, interpMode=RCIM_Cubic, leaveTangent=-3.5528, tangentMode=RCTM_User, time=1, value=0.25}
         ]}}}
      CurveIndex=$Particles.PartitionParticles.Normalized Index In Partition
      Scale Curve=22
     ))
    Write to Module Namespace=true
   )
  ]
  rendererClasses [&NiagaraRibbonRendererProperties, &NiagaraSpriteRendererProperties]
  renderers [
   {rendererClass=&NiagaraRibbonRendererProperties}
   {rendererClass=&NiagaraSpriteRendererProperties, rendererIndex=1}
  ]
 }
 Electrical_Current_Secondary {
  emitterUpdateScript [
   &EmitterState()
   &SpawnBurst_Instantaneous(Spawn Count=20, Use Spawn Probability=false, UseLoopCountLimit=true)
  ]
  particleSpawnScript [
   &InitializeParticle(
    Color=<rgba 3.2195 0.4792 4 2>
    Lifetime=2
    Ribbon Width=12
    Ribbon Width Mode=::NewEnumerator1
    UsePositionOffset=false
    Write Material Random=true
   )
   &PartitionParticles(
    Normalized Index In Partition Includes One=true
    Particles Per Partition=20
    Partition Count Mode:ENiagara_InfiniteLoopDuration=::NewEnumerator1
    Partition Mode:ENiagara_PartitionMode=::NewEnumerator1
    Write Partition Index to Ribbon Link Order:ENiagara_AttributeSamplingApplyOutput=::NewEnumerator0
    Write Partition to Ribbon ID=::NewEnumerator0
   )
   &DynamicMaterialParameters(
    Index 0 Param 2=1
    Index 0 Param 3=0.175
    Module.Param0WriteEnabled=true
    Module.Param1WriteEnabled=true
    Module.Param2WriteEnabled=true
    Module.Param3WriteEnabled=true
   )
  ]
  particleUpdateScript [
   &ParticleState(
    Kill Particles When Lifetime Has Elapsed=false
    Let Infinitely Lived Particles Die When Emitter Deactivates=true
    Loop Particles Lifetime=true
   )
   &SkeletalMeshLocation(
    MaskSpawnGroup=false
    Mesh Sampling Type=::NewEnumerator5
    Orientation Sampling=::NewEnumerator1
    Skeletal Mesh=$User.SkeletalMesh
    Socket Sampling Mode=::NewEnumerator5
    Write To Particle Position=false
   )
   &SkeletalMeshLocation(
    Bone / Socket Index=1
    MaskSpawnGroup=false
    Mesh Sampling Type=::NewEnumerator5
    Orientation Sampling=::NewEnumerator1
    Skeletal Mesh=$User.SkeletalMesh
    Socket Sampling Mode=::NewEnumerator5
    Write To Particle Position=false
   ) {module=SkeletalMeshLocation001}
   /SetVariables(Particles.MidPoint:Vector3f=&Add_Vector(
     A=&Lerp_Vector(A=$Particles.SkeletalMeshLocation.SampledPosition, B=$Particles.SkeletalMeshLocation001.SampledPosition, Alpha=0.5)
     B=&RandomRangeVector(Minimum=<-2 -2 -2>, Maximum=<2 2 2>, Randomness Mode=::NewEnumerator1, Fixed Random Seed=true)
    )) {isSetParameters=true}
   &self:BezierSpline(
    EndHandle:Vector3f=&Lerp_Vector(A=$Particles.SkeletalMeshLocation001.SampledPosition, B=$Particles.MidPoint, Alpha=0.5)
    EndPosition:Vector3f=$Particles.SkeletalMeshLocation001.SampledPosition
    PositionAlongSpline:NiagaraFloat=$Particles.PartitionParticles.Normalized Index In Partition
    StartHandle:Vector3f=&Lerp_Vector(A=$Particles.SkeletalMeshLocation.SampledPosition, B=$Particles.MidPoint, Alpha=0.5)
    StartPosition:Vector3f=$Particles.SkeletalMeshLocation.SampledPosition
   ) {module=BezierSpline}
   /SetVariables(Particles.Position:Vector3f=$Particles.Spline_Position) {isSetParameters=true, module=SetVariables_#2}
   &ScaleRibbonWidth(Ribbon Width Scale=&FloatFromCurve(
     FloatCurve=DataInterface() {propertyValues {Curve {keys [
         {arriveTangent=3.69, interpMode=RCIM_Cubic, leaveTangent=3.69, tangentMode=RCTM_User, time=0, value=0.25}
         {interpMode=RCIM_Cubic, tangentMode=RCTM_User, time=0.5, value=1}
         {arriveTangent=-3.5528, interpMode=RCIM_Cubic, leaveTangent=-3.5528, tangentMode=RCTM_User, time=1, value=0.25}
        ]}}}
     CurveIndex=$Particles.RibbonLinkOrder
    ))
  ]
  rendererClasses [&NiagaraRibbonRendererProperties]
  renderers [{rendererClass=&NiagaraRibbonRendererProperties}]
 }
 Electrical_Current_Secondary001 {
  emitterUpdateScript [
   &EmitterState()
   &SpawnBurst_Instantaneous(Spawn Count=20, Use Spawn Probability=false, UseLoopCountLimit=true)
  ]
  particleSpawnScript [
   &InitializeParticle(
    Color=<rgba 6.4021 0.7917 8 1>
    Lifetime=1.25
    Ribbon Width=15
    Ribbon Width Mode=::NewEnumerator1
    UsePositionOffset=false
    Write Material Random=true
   )
   &PartitionParticles(
    Normalized Index In Partition Includes One=true
    Particles Per Partition=20
    Partition Count Mode:ENiagara_InfiniteLoopDuration=::NewEnumerator1
    Partition Mode:ENiagara_PartitionMode=::NewEnumerator1
    Write Partition Index to Ribbon Link Order:ENiagara_AttributeSamplingApplyOutput=::NewEnumerator0
    Write Partition to Ribbon ID=::NewEnumerator0
   )
   &DynamicMaterialParameters(
    Index 0 Param 2=1
    Index 0 Param 3=0.25
    Module.Param0WriteEnabled=true
    Module.Param1WriteEnabled=true
    Module.Param2WriteEnabled=true
    Module.Param3WriteEnabled=true
   )
  ]
  particleUpdateScript [
   &ParticleState(
    Kill Particles When Lifetime Has Elapsed=false
    Let Infinitely Lived Particles Die When Emitter Deactivates=true
    Loop Particles Lifetime=true
   )
   &SkeletalMeshLocation(
    MaskSpawnGroup=false
    Mesh Sampling Type=::NewEnumerator5
    Orientation Sampling=::NewEnumerator1
    Skeletal Mesh=$User.SkeletalMesh
    Socket Sampling Mode=::NewEnumerator5
    Write To Particle Position=false
   )
   &SkeletalMeshLocation(
    Bone / Socket Index=1
    MaskSpawnGroup=false
    Mesh Sampling Type=::NewEnumerator5
    Orientation Sampling=::NewEnumerator1
    Skeletal Mesh=$User.SkeletalMesh
    Socket Sampling Mode=::NewEnumerator5
    Write To Particle Position=false
   ) {module=SkeletalMeshLocation001}
   /SetVariables(Particles.MidPoint:Vector3f=&Add_Vector(
     A=&Lerp_Vector(A=$Particles.SkeletalMeshLocation.SampledPosition, B=$Particles.SkeletalMeshLocation001.SampledPosition, Alpha=0.5)
     B=&RandomRangeVector(Minimum=<-2 -2 -2>, Maximum=<2 2 2>, Randomness Mode=::NewEnumerator1, Fixed Random Seed=true)
    )) {isSetParameters=true}
   &self:BezierSpline(
    EndHandle:Vector3f=&Lerp_Vector(A=$Particles.SkeletalMeshLocation001.SampledPosition, B=$Particles.MidPoint, Alpha=0.5)
    EndPosition:Vector3f=$Particles.SkeletalMeshLocation001.SampledPosition
    PositionAlongSpline:NiagaraFloat=$Particles.PartitionParticles.Normalized Index In Partition
    StartHandle:Vector3f=&Lerp_Vector(A=$Particles.SkeletalMeshLocation.SampledPosition, B=$Particles.MidPoint, Alpha=0.5)
    StartPosition:Vector3f=$Particles.SkeletalMeshLocation.SampledPosition
   ) {module=BezierSpline}
   /SetVariables(Particles.Position:Vector3f=$Particles.Spline_Position) {isSetParameters=true, module=SetVariables_#2}
   &ScaleRibbonWidth(Ribbon Width Scale=&FloatFromCurve(
     FloatCurve=DataInterface() {propertyValues {Curve {keys [
         {arriveTangent=3.69, interpMode=RCIM_Cubic, leaveTangent=3.69, tangentMode=RCTM_User, time=0, value=0.25}
         {interpMode=RCIM_Cubic, tangentMode=RCTM_User, time=0.5, value=1}
         {arriveTangent=-3.5528, interpMode=RCIM_Cubic, leaveTangent=-3.5528, tangentMode=RCTM_User, time=1, value=0.25}
        ]}}}
     CurveIndex=$Particles.RibbonLinkOrder
    ))
  ]
  rendererClasses [&NiagaraRibbonRendererProperties]
  renderers [{rendererClass=&NiagaraRibbonRendererProperties}]
 }
 Idle_Circle_Dust {
  emitterUpdateScript [
   &EmitterState()
   &SpawnRate(
    SpawnRate=&Multiply_Float(
     A=&MakeCustomFloatFromBool(Bool=$User.bIsInWater, True Float=0, False Float=1)
     B=&FloatFromCurve(
      FloatCurve=DataInterface() {propertyValues {Curve {keys [
          {interpMode=RCIM_Cubic, tangentMode=RCTM_User, time=0, value=1}
          {arriveTangent=-5.661, interpMode=RCIM_Cubic, leaveTangent=-5.661, tangentMode=RCTM_User, time=0.5, value=0}
         ]}}}
      CurveIndex=$System.Timeline.NormalizedTime
      Scale Curve=80
     )
    )
    Use Spawn Probability=false
   )
  ]
  particleSpawnScript [
   &InitializeParticle(
    Color=&UniformRangedLinearColor(Minimum=<rgba 0.1823 0.1561 0.1139 0.125>, Maximum=<rgba 0.2083 0.183 0.1421 0.15>)
    Lifetime Max=1
    Lifetime Min=0.5
    Lifetime Mode=::NewEnumerator1
    Sprite Rotation Mode=::NewEnumerator2
    Sprite Size Mode=::NewEnumerator1
    Uniform Sprite Size Max=70
    Uniform Sprite Size Min=60
    UsePositionOffset=false
    Write Material Random=true
   )
   &ShapeLocation(
    Cylinder Height=5
    Cylinder Height Midpoint=0
    Cylinder Radius=80
    MaskSpawnGroup=false
    Non Uniform Scale=<0.5 1 1>
    Offset=&RotateVectorByQuaternion(Vector To Rotate=<0 50 0>, Quaternion=$Engine.Owner.Rotation)
    Offset Mode=::NewEnumerator0
    Override Seed=false
    Shape Origin=$Engine.Owner.Position
    Shape Primitive=::NewEnumerator1
    Surface Only=true
    Use Endcaps In Surface Only Mode=false
    UseOwnerScale=true
   )
   &AddVelocity(Module.Use Falloff Distance=false, Origin Offset=<0 50 0>, Override Seed=false, Velocity Mode=::NewEnumerator1, Velocity Speed=750)
   &CameraOffset(Camera Offset Amount=5)
  ]
  particleUpdateScript [
   &ParticleState()
   &AccelerationForce(Acceleration=&MakeVector(
     X=0
     Y=0
     Z=&FloatFromCurve(
      FloatCurve=DataInterface() {propertyValues {Curve {keys [{interpMode=RCIM_Cubic, tangentMode=RCTM_User, time=0, value=0}, {interpMode=RCIM_Cubic, tangentMode=RCTM_User, time=1, value=1}]}}}
      Scale Curve=1000
     )
    ))
   &Drag(Drag=10, Use Linear Drag=true, Use Rotational Drag=false)
   &SolveForcesAndVelocity(Clamp Velocity=false, Limit Acceleration=false)
   &ScaleSpriteSize(Uniform Curve Sprite Scale=DataInterface() {propertyValues {Curve {keys [
        {interpMode=RCIM_Cubic, tangentMode=RCTM_User, time=0, value=0.5}
        {arriveTangent=1.5396, interpMode=RCIM_Cubic, leaveTangent=1.5396, tangentMode=RCTM_User, time=1, value=2}
       ]}}})
   /SetVariables(Particles.SubImageIndex=&FloatFromCurve(FloatCurve=DataInterface() {propertyValues {Curve {keys [
         {arriveTangent=29.8623, interpMode=RCIM_Cubic, leaveTangent=29.8623, tangentMode=RCTM_User, time=-0.0046, value=5.0556}
         <at 1.0011 22.1689 RCIM_Cubic>
        ]}}})) {isSetParameters=true, module=SetVariables_#3}
   &DynamicMaterialParameters(
    Index 0 Param 1=&FloatFromCurve(FloatCurve=DataInterface() {propertyValues {Curve {keys [<at -0.0046 1.0113 RCIM_Cubic>,<at 1.0011 0.002>]}}})
    Index 0 Param 4=&FloatFromCurve(FloatCurve=DataInterface() {propertyValues {Curve {keys [<at 0 0 RCIM_Cubic>,<at 1 0.5>]}}})
    Module.Param0WriteEnabled=true
    Module.Param1WriteEnabled=false
    Module.Param2WriteEnabled=false
    Module.Param3WriteEnabled=true
   ) {module=DynamicMatParams}
   &ScaleColor(
    Scale Alpha=&FloatFromCurve(
     FloatCurve=DataInterface() {propertyValues {Curve {keys [
         {interpMode=RCIM_Cubic, tangentMode=RCTM_User, time=0, value=0}
         <at 0.05 1>
         {arriveTangent=-1.0493, interpMode=RCIM_Cubic, leaveTangent=-1.0493, time=0.5, value=1}
         <at 1.0011 0.002>
        ]}}}
     Scale Curve=&FNTimeOfDayScalar(Night Scale=1.5, Morning Scale=0.7, Evening Scale=0.7)
    )
    ScaleA=true
    ScaleRGB=false
   )
  ]
  rendererClasses [&NiagaraSpriteRendererProperties]
  renderers [{rendererClass=&NiagaraSpriteRendererProperties}]
 }
 SmokeExhaust {
  emitterUpdateScript [
   &EmitterState()
   &SpawnPerUnit(
    Module.Use Max Movement Threshold=true
    Module.Use Movement Tolerance=true
    Movement Tolerance=1
    Spawn Probability=1
    Spawn Spacing:NiagaraFloat=&Multiply_Float(A=&MakeCustomFloatFromBool(Bool=$User.bIsInWater, True Float=0, False Float=1), B=15)
    Use Spawn Probability=false
   )
  ]
  particleSpawnScript [
   &InitializeParticle(
    Color=&UniformRangedLinearColor(Minimum=<rgba 0.1823 0.1561 0.1139 0.14>, Maximum=<rgba 0.2083 0.183 0.1421 0.16>)
    Lifetime Max=1.5
    Lifetime Mode=::NewEnumerator1
    Sprite Rotation Mode=::NewEnumerator2
    Sprite Size Mode=::NewEnumerator1
    Uniform Sprite Size Max=40
    Uniform Sprite Size Min=30
    UsePositionOffset=false
    Write Material Random=true
   )
   &AddVelocity(Velocity=<0 -750 100>)
   &InheritVelocity(Inherited Velocity Speed Limit=1000, Speed Limit=true, UseSpeedThreshold=false)
   &ApplyInitialForces()
   &CameraOffset(Camera Offset Amount=5)
  ]
  particleUpdateScript [
   &ParticleState()
   &Drag(Drag=7, Use Linear Drag=true, Use Rotational Drag=true)
   &AccelerationForce(Acceleration=&MakeVector(
     X=0
     Y=0
     Z=&FloatFromCurve(
      FloatCurve=DataInterface() {propertyValues {Curve {keys [{interpMode=RCIM_Cubic, tangentMode=RCTM_User, time=0, value=0}, {interpMode=RCIM_Cubic, tangentMode=RCTM_User, time=1, value=1}]}}}
      Scale Curve=1000
     )
    ))
   &SolveForcesAndVelocity(Clamp Velocity=false, Limit Acceleration=false)
   &ScaleSpriteSize(Uniform Curve Sprite Scale=DataInterface() {propertyValues {Curve {keys [
        {arriveTangent=0.3507, interpMode=RCIM_Cubic, leaveTangent=0.3507, tangentMode=RCTM_User, time=0, value=1.5}
        {arriveTangent=3.5211, interpMode=RCIM_Cubic, leaveTangent=3.5211, tangentMode=RCTM_User, time=1, value=4}
       ]}}})
   /SetVariables(Particles.SubImageIndex=&FloatFromCurve(FloatCurve=DataInterface() {propertyValues {Curve {keys [
         {arriveTangent=29.8623, interpMode=RCIM_Cubic, leaveTangent=29.8623, tangentMode=RCTM_User, time=-0.0046, value=5.0556}
         <at 1.0011 22.1689 RCIM_Cubic>
        ]}}})) {isSetParameters=true, module=SetVariables_#3}
   &DynamicMaterialParameters(
    Index 0 Param 1=&FloatFromCurve(FloatCurve=DataInterface() {propertyValues {Curve {keys [<at -0.0046 1.0113 RCIM_Cubic>,<at 1.0011 0.002>]}}})
    Index 0 Param 4=&FloatFromCurve(FloatCurve=DataInterface() {propertyValues {Curve {keys [<at 0 0 RCIM_Cubic>,<at 1 0.5>]}}})
    Module.Param0WriteEnabled=true
    Module.Param1WriteEnabled=false
    Module.Param2WriteEnabled=false
    Module.Param3WriteEnabled=true
   ) {module=DynamicMatParams}
   &ScaleColor(
    Scale Alpha=&FloatFromCurve(
     FloatCurve=DataInterface() {propertyValues {Curve {keys [
         {interpMode=RCIM_Cubic, tangentMode=RCTM_User, time=0, value=0}
         <at 0.05 1>
         {arriveTangent=-1.0493, interpMode=RCIM_Cubic, leaveTangent=-1.0493, time=0.5, value=1}
         <at 1.0011 0.002>
        ]}}}
     Scale Curve=&FNTimeOfDayScalar(Night Scale=1.5, Morning Scale=0.8, Evening Scale=0.8)
    )
    ScaleA=true
    ScaleRGB=false
   )
  ]
  rendererClasses [&NiagaraSpriteRendererProperties]
  renderers [{rendererClass=&NiagaraSpriteRendererProperties}]
 }
 SmokeExhaust001 {
  bEnabled=false
  emitterUpdateScript [
   &EmitterState()
   &SpawnRate(
    SpawnRate=&Multiply_Float(A=&MakeCustomFloatFromBool(Bool=$User.bIsInWater, True Float=0, False Float=1), B=40)
    Use Spawn Probability=false
   ) {enabled=false}
   &SpawnPerUnit(
    Module.Use Max Movement Threshold=true
    Module.Use Movement Tolerance=true
    Movement Tolerance=1
    Spawn Probability=1
    Spawn Spacing:NiagaraFloat=&Multiply_Float(A=&MakeCustomFloatFromBool(Bool=$User.bIsInWater, True Float=0, False Float=1), B=8)
    Use Spawn Probability=false
   )
  ]
  particleSpawnScript [
   &PartitionParticles()
   &InitializeParticle(
    Color=&UniformRangedLinearColor(Minimum=<rgba 0.1823 0.1561 0.1139 0.25>, Maximum=<rgba 0.2083 0.183 0.1421 0.3>)
    Lifetime Max=1.5
    Lifetime Mode=::NewEnumerator1
    Sprite Rotation Mode=::NewEnumerator2
    Sprite Size Mode=::NewEnumerator1
    Uniform Sprite Size Max=40
    Uniform Sprite Size Min=30
    UsePositionOffset=false
    Write Material Random=true
   )
   &SkeletalMeshLocation(
    Bone / Socket Index=$Particles.PartitionParticles.Partition
    Bone Sampling Mode=::NewEnumerator4
    MaskSpawnGroup=false
    Skeletal Mesh=$User.SkeletalMesh
   )
   /SetVariables(Particles.Position=&MakeVector(
     X=&MakeFloatFromVector(Vector=$Particles.SkeletalMeshLocation.SampledPosition, Channel=::NewEnumerator0)
     Y=&MakeFloatFromVector(Vector=$Particles.SkeletalMeshLocation.SampledPosition, Channel=::NewEnumerator1)
     Z=&MakeFloatFromVector(Vector=Linked() {linkedVariable {name=Engine.Owner.Position, type=NiagaraPosition}}, Channel=::NewEnumerator2)
    )) {isSetParameters=true}
   &AddVelocity(Velocity=<0 -750 200>)
   &InheritVelocity(Inherited Velocity Speed Limit=1000, Speed Limit=true, UseSpeedThreshold=false)
   &ApplyInitialForces()
   &CameraOffset(Camera Offset Amount=5)
  ]
  particleUpdateScript [
   &ParticleState()
   &Drag(Drag=7, Use Linear Drag=true, Use Rotational Drag=true)
   &AccelerationForce(Acceleration=&MakeVector(
     X=0
     Y=0
     Z=&FloatFromCurve(
      FloatCurve=DataInterface() {propertyValues {Curve {keys [{interpMode=RCIM_Cubic, tangentMode=RCTM_User, time=0, value=0}, {interpMode=RCIM_Cubic, tangentMode=RCTM_User, time=1, value=1}]}}}
      Scale Curve=1000
     )
    ))
   &SolveForcesAndVelocity(Clamp Velocity=false, Limit Acceleration=false)
   &ScaleSpriteSize(Uniform Curve Sprite Scale=DataInterface() {propertyValues {Curve {keys [
        {arriveTangent=0.3507, interpMode=RCIM_Cubic, leaveTangent=0.3507, tangentMode=RCTM_User, time=0, value=0.65}
        {interpMode=RCIM_Cubic, tangentMode=RCTM_User, time=1, value=2.5}
       ]}}})
   /SetVariables(Particles.SubImageIndex=&FloatFromCurve(FloatCurve=DataInterface() {propertyValues {Curve {keys [
         {arriveTangent=29.8623, interpMode=RCIM_Cubic, leaveTangent=29.8623, tangentMode=RCTM_User, time=-0.0046, value=5.0556}
         <at 1.0011 22.1689 RCIM_Cubic>
        ]}}})) {isSetParameters=true, module=SetVariables_#3}
   &DynamicMaterialParameters(
    Index 0 Param 1=&FloatFromCurve(FloatCurve=DataInterface() {propertyValues {Curve {keys [<at -0.0046 1.0113 RCIM_Cubic>,<at 1.0011 0.002>]}}})
    Index 0 Param 4=&FloatFromCurve(FloatCurve=DataInterface() {propertyValues {Curve {keys [<at 0 0 RCIM_Cubic>,<at 1 0.5>]}}})
    Module.Param0WriteEnabled=true
    Module.Param1WriteEnabled=false
    Module.Param2WriteEnabled=false
    Module.Param3WriteEnabled=true
   ) {module=DynamicMatParams}
   &ScaleColor(
    Scale Alpha=&FloatFromCurve(
     FloatCurve=DataInterface() {propertyValues {Curve {keys [
         {interpMode=RCIM_Cubic, tangentMode=RCTM_User, time=0, value=0}
         <at 0.05 1>
         {arriveTangent=-1.0493, interpMode=RCIM_Cubic, leaveTangent=-1.0493, time=0.5, value=1}
         <at 1.0011 0.002>
        ]}}}
     Scale Curve=&FNTimeOfDayScalar(Night Scale=1.5, Morning Scale=1, Evening Scale=1)
    )
    ScaleA=true
    ScaleRGB=false
   )
  ]
  rendererClasses [&NiagaraSpriteRendererProperties]
  renderers [{rendererClass=&NiagaraSpriteRendererProperties}]
 }
 SmokeGroundRings {
  emitterUpdateScript [
   &EmitterState()
   &SpawnRate(
    SpawnRate=&Multiply_Float(
     A=&MakeCustomFloatFromBool(Bool=$User.bIsInWater, True Float=0, False Float=1)
     B=&FloatFromCurve(
      FloatCurve=DataInterface() {propertyValues {Curve {keys [
          {interpMode=RCIM_Cubic, tangentMode=RCTM_User, time=0, value=1}
          {arriveTangent=-5.661, interpMode=RCIM_Cubic, leaveTangent=-5.661, tangentMode=RCTM_User, time=0.5, value=0}
         ]}}}
      CurveIndex=$System.Timeline.NormalizedTime
      Scale Curve=4
     )
    )
    Use Spawn Probability=false
   )
  ]
  particleSpawnScript [
   &InitializeParticle(
    Color=<rgba 0.4427 0.4427 0.4427 0.5>
    Lifetime Max=1.25
    Lifetime Mode=::NewEnumerator1
    Sprite Size=<7 10>
    Sprite Size Mode=::NewEnumerator4
    UsePositionOffset=false
    Write Material Random=true
   )
   &SystemLocation(MaskSpawnGroup=false, Offset=<0 50 5>)
   &SpriteFacingAndAlignment(Sprite Facing=<0 0 1>, WriteAlignment=true, WriteFacing=true)
  ]
  particleUpdateScript [
   &ParticleState()
   &DynamicMaterialParameters(
    Index 0 Param 1=&FloatFromCurve(FloatCurve=DataInterface() {propertyValues {Curve {keys [<at 0 0.5>,<at 1 2>]}}})
    Module.Param0WriteEnabled=true
    Module.Param1WriteEnabled=false
    Module.Param2WriteEnabled=false
    Module.Param3WriteEnabled=false
   )
   &ScaleSpriteSize(
    Uniform Curve Scale=50
    Uniform Curve Sprite Scale=DataInterface() {propertyValues {Curve {keys [
        {arriveTangent=1.0307, interpMode=RCIM_Cubic, leaveTangent=1.0307, tangentMode=RCTM_User, time=0, value=0}
        {arriveTangent=0.0224, interpMode=RCIM_Cubic, leaveTangent=0.0224, tangentMode=RCTM_User, time=1, value=1}
       ]}}}
   )
   &CameraOffset(Camera Offset Amount=20)
   &ScaleColor(
    Scale Alpha=&FloatFromCurve(
     FloatCurve=DataInterface() {propertyValues {Curve {keys [
         {tangentMode=RCTM_User, time=0, value=1}
         {arriveTangent=-0.0405, interpMode=RCIM_Cubic, leaveTangent=-0.0405, tangentMode=RCTM_User, time=0.7009, value=1}
         {arriveTangent=-8.0089, interpMode=RCIM_Cubic, leaveTangent=-8.0089, tangentMode=RCTM_User, time=1, value=0}
        ]}}}
     Scale Curve=&FloatFromCurve(
      FloatCurve=DataInterface() {propertyValues {Curve {keys [
          {interpMode=RCIM_Cubic, tangentMode=RCTM_User, time=0, value=1}
          {arriveTangent=-11.478, interpMode=RCIM_Cubic, leaveTangent=-11.478, tangentMode=RCTM_User, time=0.25, value=0}
         ]}}}
      CurveIndex=$System.Timeline.NormalizedTime
     )
    )
    Scale RGB=&MakeVectorFromLinearColorRGB(Color=&FNTimeOfDayColor(
      Night Color=<rgba 0.2123 0.2489 0.27 1>
      Morning Color=<rgba 0.445 0.2629 0.1507 1>
      Day Color=<rgba 0.3626 0.4008 0.4271 1>
      Evening Color=<rgba 0.599 0.2667 0.2028 1>
     ))
    ScaleA=true
    ScaleRGB=true
   )
  ]
  rendererClasses [&NiagaraSpriteRendererProperties]
  renderers [{rendererClass=&NiagaraSpriteRendererProperties}]
 }
 Thruster_Fire_Mesh {
  emitterUpdateScript [
   &EmitterState(
    Delay First Loop Only=true
    Inactive Response=::NewEnumerator1
    Life Cycle Mode=::NewEnumerator1
    Loop Delay=0.2
    Loop Duration=1
    UseLoopDelay=true
   )
   &SpawnRate(SpawnRate=20, Use Spawn Probability=false)
  ]
  particleSpawnScript [
   &PartitionParticles()
   &InitializeParticle(
    Color=<rgba 1 1 1 0.1>
    Lifetime Max=0.27
    Lifetime Min=0.14
    Lifetime Mode=::NewEnumerator1
    Mesh Scale=<0.225 0.225 0.225>
    Mesh Scale Mode=::NewEnumerator4
    Sprite Size Mode=::NewEnumerator3
    Uniform Sprite Size=14
    UsePositionOffset=false
    Write Material Random=true
   )
   &InitialMeshOrientation(
    Mesh Axis To Orient=<0 1 0>
    Mesh Orientation Mode=::NewEnumerator2
    Orientation Vector=&Lerp_Vector(A=<0 1 0>, B=<0 -1 0>, Alpha=&MakeFloatFromInt(Int=$Particles.PartitionParticles.Partition))
    Rotation=&RandomRangeVector(Minimum=<0 -1 0>, Maximum=<0 1 0>)
    Use Rotation Vector=true
   )
   &CameraOffset(Camera Offset Amount=0.5)
  ]
  particleUpdateScript [
   &ParticleState()
   &SkeletalMeshLocation(
    Bone / Socket Index=$Particles.PartitionParticles.Partition
    Bone Sampling Mode=::NewEnumerator4
    MaskSpawnGroup=false
    Sampled Position Offset=&Lerp_Vector(
     A=<0.25 28 12.75>
     B=<-0.25 -28 -12.75>
     Alpha=&MakeFloatFromInt(Int=$Particles.PartitionParticles.Partition)
    )
    Skeletal Mesh=$User.SkeletalMesh
   )
   &FN_Sprite_Attributes(
    Alignment:Vector3f=<1 0 0>
    Camera Offset=3
    Color=<rgba 2 0.3882 0 1>
    Create Alignment Attribute=true
    Create Camera Offset Attribute=true
    Create Color Attribute=true
    Create Facing Attribute=true
    Create SpriteSize Attribute=true
    Facing:Vector3f=&RotateVectorByQuaternion(
     Vector To Rotate=&Lerp_Vector(A=<0 1 0>, B=<0 -1 0>, Alpha=&MakeFloatFromInt(Int=$Particles.PartitionParticles.Partition))
     Quaternion=$Particles.SkeletalMeshLocation.SampledRotation
    )
    SpriteSize=&MakeVector2D(
     X=18
     Y=&FloatFromCurve(
      FloatCurve=DataInterface() {propertyValues {Curve {keys [
          {interpMode=RCIM_Cubic, tangentMode=RCTM_User, time=0, value=1}
          {arriveTangent=-0.0702, interpMode=RCIM_Cubic, leaveTangent=-0.0702, tangentMode=RCTM_User, time=0.25, value=2}
          {arriveTangent=-4.5485, interpMode=RCIM_Cubic, leaveTangent=-4.5485, tangentMode=RCTM_User, time=1, value=1}
         ]}}}
      CurveIndex=$System.Timeline.NormalizedTime
      Scale Curve=60
     )
    )
    Write to Module Namespace=true
   )
   &SpriteFacingAndAlignment(
    Sprite Alignment=&RotateVectorByQuaternion(Vector To Rotate=<0 1 0>, Quaternion=$Particles.SkeletalMeshLocation.SampledRotation)
    WriteAlignment=true
    WriteFacing=true
   )
   &ScaleColor(
    Scale Alpha=&FloatFromCurve(FloatCurve=DataInterface() {propertyValues {Curve {keys [
         {tangentMode=RCTM_User, time=0, value=1}
         {arriveTangent=-0.0405, interpMode=RCIM_Cubic, leaveTangent=-0.0405, tangentMode=RCTM_User, time=0.7009, value=1}
         {arriveTangent=-8.0089, interpMode=RCIM_Cubic, leaveTangent=-8.0089, tangentMode=RCTM_User, time=1, value=0}
        ]}}})
    ScaleA=true
    ScaleRGB=true
   )
   &ScaleMeshSize(Scale Factor=&MakeVector(
     X=&FloatFromCurve(FloatCurve=DataInterface() {propertyValues {Curve {keys [
          {tangentMode=RCTM_User, time=0, value=1}
          {arriveTangent=-0.0405, interpMode=RCIM_Cubic, leaveTangent=-0.0405, tangentMode=RCTM_User, time=0.7009, value=1}
          {arriveTangent=-8.0089, interpMode=RCIM_Cubic, leaveTangent=-8.0089, tangentMode=RCTM_User, time=1, value=0}
         ]}}})
     Y=&FloatFromCurve(
      FloatCurve=DataInterface() {propertyValues {Curve {keys [
          {interpMode=RCIM_Cubic, tangentMode=RCTM_User, time=0, value=0.75}
          {arriveTangent=-0.0702, interpMode=RCIM_Cubic, leaveTangent=-0.0702, tangentMode=RCTM_User, time=0.25, value=1.75}
          {arriveTangent=-4.5485, interpMode=RCIM_Cubic, leaveTangent=-4.5485, tangentMode=RCTM_User, time=1, value=1}
         ]}}}
      CurveIndex=$System.Timeline.NormalizedTime
     )
     Z=&FloatFromCurve(FloatCurve=DataInterface() {propertyValues {Curve {keys [
          {tangentMode=RCTM_User, time=0, value=1}
          {arriveTangent=-0.0405, interpMode=RCIM_Cubic, leaveTangent=-0.0405, tangentMode=RCTM_User, time=0.7009, value=1}
          {arriveTangent=-8.0089, interpMode=RCIM_Cubic, leaveTangent=-8.0089, tangentMode=RCTM_User, time=1, value=0}
         ]}}})
    ))
   /SetVariables(Particles.VerticalGlowColor:LinearColor=&ColorFromCurve(ColorCurve=DataInterface() {propertyValues {
       AlphaCurve {keys [<at 0.3162 0.025 RCIM_Cubic>,<at 1 0 RCIM_Cubic>]}
       BlueCurve {keys [<at 0 0 RCIM_Cubic>,<at 1 0 RCIM_Cubic>]}
       GreenCurve {keys [<at 0 0.1617 RCIM_Cubic>,<at 1 0.2743 RCIM_Cubic>]}
       RedCurve {keys [<at 0 1 RCIM_Cubic>,<at 1 1 RCIM_Cubic>]}
      }})) {isSetParameters=true}
  ]
  rendererClasses [&NiagaraMeshRendererProperties, &NiagaraSpriteRendererProperties]
  renderers [
   {rendererClass=&NiagaraMeshRendererProperties, rendererIndex=0}
   {rendererClass=&NiagaraSpriteRendererProperties, rendererIndex=1}
   {rendererClass=&NiagaraSpriteRendererProperties, rendererIndex=2}
  ]
 }
 Wind_Streaks {
  emitterUpdateScript [
   &EmitterState()
   &SpawnRate(
    SpawnRate=&Multiply_Float(
     A=&MakeCustomFloatFromBool(Bool=$User.bIsInWater, True Float=0, False Float=1)
     B=&FloatFromCurve(
      FloatCurve=DataInterface() {propertyValues {Curve {keys [
          {interpMode=RCIM_Cubic, tangentMode=RCTM_User, time=0, value=1}
          {arriveTangent=-5.661, interpMode=RCIM_Cubic, leaveTangent=-5.661, tangentMode=RCTM_User, time=0.5, value=0}
         ]}}}
      CurveIndex=$System.Timeline.NormalizedTime
      Scale Curve=50
     )
    )
    Use Spawn Probability=false
   )
  ]
  particleSpawnScript [
   &InitializeParticle(
    Color=<rgba 0.8854 0.8854 0.8854 0.65>
    Lifetime Max=0.4
    Lifetime Min=0.35
    Lifetime Mode=::NewEnumerator1
    Sprite Size Max=<8 80>
    Sprite Size Min=<5 50>
    Sprite Size Mode=::NewEnumerator2
    UsePositionOffset=false
    Write Material Random=true
   )
   &ShapeLocation(
    Cylinder Height=0
    Cylinder Radius=1
    MaskSpawnGroup=false
    Offset=<0 50 0>
    Offset Mode=::NewEnumerator0
    Override Seed=false
    Shape Primitive=::NewEnumerator1
    Surface Only=false
    UseOwnerScale=true
   )
   &AddVelocity(Velocity=&MakeVector(X=0, Y=0, Z=&UniformRangedFloat(Minimum=0, Maximum=100)))
   &AddVelocityFromPoint(Module.Use Falloff Distance=false, Origin Offset=<0 50 0>, Velocity Strength=300)
   &DynamicMaterialParameters(
    Index 0 Param 1=&UniformRangedFloat(Minimum=0, Maximum=1)
    Index 0 Param 2=&UniformRangedFloat(Minimum=1, Maximum=2)
    Index 0 Param 3=1
    Index 0 Param 4=1
    Module.Param0WriteEnabled=true
    Module.Param1WriteEnabled=true
    Module.Param2WriteEnabled=true
    Module.Param3WriteEnabled=true
   )
  ]
  particleUpdateScript [
   &ParticleState()
   &Drag(Drag=1.5, Use Linear Drag=true, Use Rotational Drag=false)
   &SolveForcesAndVelocity(Clamp Velocity=false, Limit Acceleration=false)
   &ScaleColor(
    Scale Alpha=&FloatFromCurve(
     FloatCurve=DataInterface() {propertyValues {Curve {keys [<at 0 0 RCIM_Cubic>, {arriveTangent=0.002, interpMode=RCIM_Cubic, leaveTangent=0.002, time=0.5, value=1}, <at 1.0011 0.002 RCIM_Cubic>]}}}
     Scale Curve=&FNTimeOfDayScalar(Night Scale=0.75, Morning Scale=1, Day Scale=0.9, Evening Scale=1.2)
    )
    Scale RGB=&MakeVectorFromLinearColorRGB(Color=&FNTimeOfDayColor(
      Night Color=<rgba 0.2123 0.2489 0.27 1>
      Morning Color=<rgba 0.445 0.2629 0.1507 1>
      Day Color=<rgba 0.3626 0.4008 0.4271 1>
      Evening Color=<rgba 0.599 0.2667 0.2028 1>
     ))
    ScaleA=true
    ScaleRGB=true
   )
  ]
  rendererClasses [&NiagaraSpriteRendererProperties]
  renderers [{rendererClass=&NiagaraSpriteRendererProperties}]
 }
}
summary {
 systemName=NS_PlotTwist_Emote
 emitters [
  {
   emitterName=Electrical_Current
   rendererClasses [&NiagaraRibbonRendererProperties, &NiagaraSpriteRendererProperties]
  }
  {
   emitterName=Electrical_Connector_Flares
   rendererClasses [&NiagaraSpriteRendererProperties, &NiagaraLightRendererProperties]
  }
  {emitterName=SmokeExhaust, rendererClasses [&NiagaraSpriteRendererProperties]}
  {
   emitterName=Thruster_Fire_Mesh
   rendererClasses [&NiagaraMeshRendererProperties, &NiagaraSpriteRendererProperties]
  }
  {emitterName=Idle_Circle_Dust, rendererClasses [&NiagaraSpriteRendererProperties]}
  {emitterName=SmokeGroundRings, rendererClasses [&NiagaraSpriteRendererProperties]}
  {emitterName=Wind_Streaks, rendererClasses [&NiagaraSpriteRendererProperties]}
  {bEnabled=false, emitterName=SmokeExhaust001, rendererClasses [&NiagaraSpriteRendererProperties]}
  {emitterName=Electrical_Current_Secondary, rendererClasses [&NiagaraRibbonRendererProperties]}
  {emitterName=Electrical_Current_Secondary001, rendererClasses [&NiagaraRibbonRendererProperties]}
 ]
 userVariables [
  User.bIsInWater {type=NiagaraBool}
  User.SkeletalMesh {
   type=NiagaraDataInterfaceSkeletalMesh
   defaultValue {
    struct=NiagaraExt_VariableValue_DataInterface
    value {dataInterface=&self:NiagaraDataInterfaceSkeletalMesh_0}
   }
  }
 ]
}
